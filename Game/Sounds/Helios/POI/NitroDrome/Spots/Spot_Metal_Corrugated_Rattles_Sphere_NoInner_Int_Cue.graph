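# Indoor ambient one-shot cue for the NitroDrome POI that plays intermittent corrugated metal rattles.
# An endless loop waits a random delay, then passes through a quality switch with only its first two of four slots connected.
# Those share a modulator varying pitch and volume over an equally weighted random pick of seven rattle waves.
# It uses a NoInner metal-creak attenuation and overrides concurrency to cap simultaneous voices.
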SoundCue /Game/Sounds/Helios/POI/NitroDrome/Spots/Spot_Metal_Corrugated_Rattles_Sphere_NoInner_Int_Cue
root=Looping_0
nodes [
 {id=Delay_0, children [QualityLevel_0], properties {delayMax=18, delayMin=0.8}}
 {id=Looping_0, children [Delay_0]}
 {id=Modulator_0, children [Random_1], properties {pitchMin=0.8, volumeMin=0.8}}
 {id=QualityLevel_0, children [Modulator_0,Modulator_0,null,null]}
 {
  id=Random_1
  children [WavePlayer_0, WavePlayer_1, WavePlayer_2, WavePlayer_13, WavePlayer_14, WavePlayer_15, WavePlayer_16]
  properties {weights [1,1,1,1,1,1,1]}
 }
 {id=WavePlayer_0, properties {soundWaveAssetPtr=&/Spot_Metal_Corrugated_Rattles_07}}
 {id=WavePlayer_1, properties {soundWaveAssetPtr=&/Spot_Metal_Corrugated_Rattles_06}}
 {id=WavePlayer_13, properties {soundWaveAssetPtr=&/Spot_Metal_Corrugated_Rattles_04}}
 {id=WavePlayer_14, properties {soundWaveAssetPtr=&/Spot_Metal_Corrugated_Rattles_03}}
 {id=WavePlayer_15, properties {soundWaveAssetPtr=&/Spot_Metal_Corrugated_Rattles_02}}
 {id=WavePlayer_16, properties {soundWaveAssetPtr=&/Spot_Metal_Corrugated_Rattles_01}}
 {id=WavePlayer_2, properties {soundWaveAssetPtr=&/Spot_Metal_Corrugated_Rattles_05}}
]
properties {
 attenuationSettings=/Game/Sounds/Attenuation/Fort_TileAmbient_OneShot_Metal_Creak_Actor_NoInner_Int_Att
 bOverrideConcurrency=true
 maxDistance=2000
 soundClassObject=/Game/Sounds/FortSoundClassesAndModes/Fort_Ambient_OneShots_Indoors
 volumeMultiplier=0.301
 attenuationOverrides {falloffDistance=6000, stereoSpread=10000, attenuationShapeExtents {x=2000}}
 concurrencyOverrides {maxCount=2, voiceStealReleaseTime=2}
}
